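# Shadow-volume physics asset for the large female Constructor_01 player body, approximating the character with capsules, as its name suggests for shadow casting.
# Twelve single-capsule bodies cover pelvis, upper spine, arms, legs and feet, mostly on twist bones, with none for head or hands.
# Constraints form a shallow tree rather than limb chains: arm capsules link directly to spine_03, while leg capsules and spine_03 link to the pelvis.
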PhysicsAsset /Game/Characters/Player/Female/Large/Bodies/F_LRG_Constructor_01/Mesh/SK_F_LRG_Constructor_01_ShadowVolumes
bodies {
 calf_twist_02_l [{shapeType=Capsule, length=41.2069, radius=7.7282, center=<-2.3286 -1.1138 -0.1616>, rotation=<rot -0.0062 87.9534 -89.9636>}]
 calf_twist_02_r [{shapeType=Capsule, length=47.8758, radius=9.0433, center=<-2.362 0.5687 0.1066>, rotation=<rot -0.0147 -90.8481 -89.8927>}]
 foot_l [{shapeType=Capsule, length=13.1659, radius=7.799, center=<-2.495 5.8526 -0.0561>, rotation=<rot -0.672 179.2464 -91.4399>}]
 foot_r [{shapeType=Capsule, length=8.4159, radius=8.7988, center=<2.7493 -5.7525 0.7073>, rotation=<rot 0.6234 1.1655 -88.5384>}]
 lowerarm_l [{shapeType=Capsule, length=35.0482, radius=5.8186, center=<17.5656 -2.5798 -1.2337>, rotation=<rot 2.7686 -102.2653 -86.1165>}]
 lowerarm_twist_01_r [{shapeType=Capsule, length=26.7452, radius=5.8107, center=<-0.8712 0.5689 1.0515>, rotation=<rot 0 -94.0739 -84.6099>}]
 pelvis [{shapeType=Capsule, length=15.5815, radius=17.6215, center=<-0.9838 -2.5194 -0.1056>, rotation=<rot -4.9681 -82.8847 -0.5914>}]
 spine_03 [{shapeType=Capsule, length=17.0126, radius=15.5716, center=<7.9856 1.0959 -0.133>, rotation=<rot 0 90 -90>}]
 thigh_twist_01_l [{shapeType=Capsule, length=30.3336, radius=11.7673, center=<7.9097 -2.3933 -2.0262>, rotation=<rot 0.1146 -82.2321 -88.8882>}]
 thigh_twist_01_r [{shapeType=Capsule, length=22.9493, radius=11.0324, center=<-8.9821 -0.1471 2.2298>, rotation=<rot -1.6368 -85.8033 -85.0445>}]
 upperarm_twist_01_l [{shapeType=Capsule, length=20.9787, radius=5.8924, center=<3.8306 -0.4087 0.4242>, rotation=<rot 0 -82.7612 -89.1886>}]
 upperarm_twist_01_r [{shapeType=Capsule, length=28.169, radius=6.2045, center=<-6.1534 1.6414 0.2233>, rotation=<rot -14.813 -92.5958 -92.3309>}]
}
constraints [
 {bone1Name=spine_03, bone2Name=pelvis, swing1LimitDegrees=45, swing2LimitDegrees=45, twistLimitDegrees=45}
 {bone1Name=lowerarm_l, bone2Name=spine_03, swing1LimitDegrees=45, swing2LimitDegrees=45, twistLimitDegrees=45}
 {bone1Name=upperarm_twist_01_l, bone2Name=spine_03, swing1LimitDegrees=45, swing2LimitDegrees=45, twistLimitDegrees=45}
 {bone1Name=lowerarm_twist_01_r, bone2Name=spine_03, swing1LimitDegrees=45, swing2LimitDegrees=45, twistLimitDegrees=45}
 {bone1Name=upperarm_twist_01_r, bone2Name=spine_03, swing1LimitDegrees=45, swing2LimitDegrees=45, twistLimitDegrees=45}
 {bone1Name=calf_twist_02_l, bone2Name=pelvis, swing1LimitDegrees=45, swing2LimitDegrees=45, twistLimitDegrees=45}
 {bone1Name=foot_l, bone2Name=pelvis, swing1LimitDegrees=45, swing2LimitDegrees=45, twistLimitDegrees=45}
 {bone1Name=thigh_twist_01_l, bone2Name=pelvis, swing1LimitDegrees=45, swing2LimitDegrees=45, twistLimitDegrees=45}
 {bone1Name=calf_twist_02_r, bone2Name=pelvis, swing1LimitDegrees=45, swing2LimitDegrees=45, twistLimitDegrees=45}
 {bone1Name=foot_r, bone2Name=pelvis, swing1LimitDegrees=45, swing2LimitDegrees=45, twistLimitDegrees=45}
 {bone1Name=thigh_twist_01_r, bone2Name=pelvis, swing1LimitDegrees=45, swing2LimitDegrees=45, twistLimitDegrees=45}
]
